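# Physics asset for the offhand axe mesh of the WingBath pickaxe, a Battle Royale melee cosmetic.
# It holds a single body on the Melee_TwoHanded_Handle bone, shaped as one short, wide capsule that is offset and rotated from the bone.
# It has no constraints, so the mesh has one simple physics volume with no jointed parts.
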PhysicsAsset /BRCosmetics/Weapons/FORT_Melee/Pickaxe_WingBath/Meshes/WingBath_Offhand/WingBath_Offhand_Axe_PhysicsAsset
bodies {Melee_TwoHanded_Handle [{shapeType=Capsule, length=13.9606, radius=13.9605, center=<-8.2141 15.9462 -3.4791>, rotation=<rot 14.8492 -90 0.055>}]}
constraints []
